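# Master additive unlit material for Niagara mesh particles that produces zoom motion blur by outputting translucent velocity, with zero emissive input.
# A previous-frame switch offsets only last frame's vertices radially from the object position, faking motion vectors.
# Particle alpha and squared vertex color weight that offset, which is scaled by delta time.
# A parameterized, clamped camera offset is added to the displacement.
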Material /Game/Effects/Niagara/Emitters/MotionBlur/Materials/M_Zoom_Blur_Mesh_Master
graph {
 properties {
  bOutputTranslucentVelocity=true
  bUsedWithNiagaraMeshParticles=true
  blendMode=BLEND_Additive
  pixelDepthOffsetMode=PDOM_Legacy
  shadingModel=MSM_Unlit
  usedShadingModels=Unlit
 }
 nodes [
  n1 = Constant()
  n2 = WorldPosition()
  n3 = ObjectPositionWS()
  n4 = ParticleColor()
  n5 = VertexColor()
  n6 = Multiply(A=@n5, B=@n5)
  n7 = Multiply(A=LinearInterpolate(B=Subtract(A=@n2.XYZ, B=@n3), Alpha=Multiply(A=@n4.A, B=@n6))) {constB=500}
  Add(
   A=Multiply(A=PreviousFrameSwitch(Current Frame=@n1, Previous Frame=@n7), B=DeltaTime())
   B=MaterialFunctionCall(Offset Amount=ScalarParameter() {defaultValue=600, parameterName=CameraOffsetAmount}) {materialFunction=/Engine/Functions/Engine_MaterialFunctions02/WorldPositionOffset/CameraOffset}.Clamped Camera Offset
  )
  MaterialFunctionCall(Emissive=Constant()) {materialFunction=&MF_FX_EmissiveAdjustments}
 ]
}
